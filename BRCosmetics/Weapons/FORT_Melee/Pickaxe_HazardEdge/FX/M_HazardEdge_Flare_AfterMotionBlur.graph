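# Additive unlit Niagara sprite material for a flare effect on the Hazard Edge pickaxe, drawn in the after-motion-blur translucency pass.
# Particle color RGB feeds a shared emissive adjustment function.
# Particle alpha times the flare texture's red channel goes through a parameterized depth fade.
# Feature-level and shading-path switches take that result, except the mobile path, which uses raw texture red without fade or particle alpha.
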Material /BRCosmetics/Weapons/FORT_Melee/Pickaxe_HazardEdge/FX/M_HazardEdge_Flare_AfterMotionBlur
graph {
 properties {
  bAutomaticallySetUsageInEditor=false
  bUsedWithNiagaraSprites=true
  blendMode=BLEND_Additive
  shadingModel=MSM_Unlit
  translucencyPass=MTP_AfterMotionBlur
  usedShadingModels=Unlit
 }
 nodes [
  n1 = ParticleColor()
  n2 = TextureSampleParameter2D() {parameterName=TexFlare, texture=/Game/Effects/Fort_Effects/Textures/Flares/T_Flare_Spikeball}
  n3 = DepthFade(Opacity=Multiply(A=@n1.A, B=@n2.R), FadeDistance=ScalarParameter() {defaultValue=25, parameterName=DepthFade_Distance})
  FeatureLevelSwitch(Default=@n3, SM5=@n3, SM6=@n3)
  MaterialFunctionCall(Emissive=@n1.RGB) {materialFunction=&MF_FX_EmissiveAdjustments}
  ShadingPathSwitch(Default=@n3, Mobile=@n2.R)
 ]
}
