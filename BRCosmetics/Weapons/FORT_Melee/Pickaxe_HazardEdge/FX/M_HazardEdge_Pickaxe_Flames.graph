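# Translucent unlit material for animated red flames on Niagara mesh particles in the Hazard Edge pickaxe effect.
# Two panning plasma textures combine with Fresnel to blend two emissive reds, tinted by particle color; mobile replaces the second texture with a constant.
# Opacity fades at silhouettes and V edges, scaled by particle alpha.
# The same noise drives world position offset along normals plus a direction parameter, masked by sine-based UV edge falloffs.
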Material /BRCosmetics/Weapons/FORT_Melee/Pickaxe_HazardEdge/FX/M_HazardEdge_Pickaxe_Flames
graph {
 properties {
  bAutomaticallySetUsageInEditor=false
  bUsedWithNiagaraMeshParticles=true
  blendMode=BLEND_Translucent
  shadingModel=MSM_Unlit
  usedShadingModels=Unlit
 }
 nodes [
  n1 = Constant3Vector() {constant=<rgba 1 0.0083 0 0>}
  n2 = Constant3Vector() {constant=<rgba 5 0.1 0.1817 1>}
  n3 = Fresnel() {baseReflectFraction=0, exponent=0.8}
  n4 = MaterialFunctionCall(Float2=TextureCoordinate()) {materialFunction=&BreakOutFloat2Components}
  n5 = Multiply(A=ScalarParameter() {defaultValue=0.4, parameterName=Tex2_Scale_U}, B=@n4.R) {constB=0.4}
  n6 = Multiply(A=ScalarParameter() {defaultValue=0.03, parameterName=Tex2_Speed_Pan_U}, B=Time()) {constB=0.03}
  n7 = Multiply(A=@n4.G, B=ScalarParameter() {defaultValue=0.2, parameterName=Tex2_Scale_V}) {constB=0.2}
  n8 = Multiply(A=ScalarParameter() {defaultValue=0.2, parameterName=Tex2_Speed_Pan_V}, B=Time()) {constB=0.2}
  n9 = TextureSampleParameter2D(UVs=AppendVector(A=Add(A=@n5, B=@n6), B=Add(A=@n7, B=@n8))) {parameterName=Texture_2, samplerType=SAMPLERTYPE_LinearColor, texture=/Game/Effects/Fort_Effects/Emotes/UltraLight/Texture/T_Plasma_A_C}
  n10 = ShadingPathSwitch(Default=@n9.B, Deferred=@n9.B, Mobile=Constant() {r=0.35}) {desc=Migrated from FeatureLevelSwitch}
  n11 = Power(Base=@n10)
  n12 = MaterialFunctionCall(Float2=TextureCoordinate()) {materialFunction=&BreakOutFloat2Components}
  n13 = Multiply(A=ScalarParameter() {defaultValue=1, parameterName=Tex1_Scale_U}, B=@n12.R)
  n14 = Multiply(A=Time(), B=ScalarParameter() {defaultValue=0.1, parameterName=Tex1_Speed_Pan_U}) {constB=0.1}
  n15 = Multiply(A=ScalarParameter() {defaultValue=0.6, parameterName=Tex1_Scale_V}, B=@n12.G) {constB=0.6}
  n16 = Multiply(A=Time(), B=ScalarParameter() {defaultValue=1, parameterName=Tex1_Speed_Pan_V})
  n17 = TextureSampleParameter2D(UVs=AppendVector(A=Add(A=@n13, B=@n14), B=Add(A=@n15, B=@n16))) {parameterName=Texture_1, samplerType=SAMPLERTYPE_LinearColor, texture=&T_Plasma_A_C}
  n18 = Saturate(None=Multiply(A=Add(A=@n11, B=@n17.R)) {constB=1.5})
  n19 = LinearInterpolate(A=Multiply(A=Add(A=@n3) {constB=0.1}, B=@n18), B=@n18, Alpha=@n3)
  n20 = Saturate(None=Power(Base=@n19) {constExponent=4})
  n21 = ParticleColor()
  MaterialFunctionCall(Emissive=Multiply(A=LinearInterpolate(A=@n1, B=@n2, Alpha=@n20), B=@n21.RGB)) {materialFunction=&MF_FX_EmissiveAdjustments}
  n22 = Fresnel() {baseReflectFraction=0, exponent=4}
  n23 = Add(A=@n19, B=@n22) {constB=0}
  n24 = Sine(None=Multiply(A=@n12.G) {constB=0.5})
  n25 = Saturate(None=Power(Base=@n24) {constExponent=1.5})
  n26 = OneMinus(None=@n22)
  Multiply(A=@n21.A, B=Multiply(A=Multiply(A=@n23, B=@n25), B=Power(Base=@n26) {constExponent=8}))
  n27 = ScalarParameter() {defaultValue=0.2, parameterName=WPO_Tex_Lerp}
  n28 = Multiply(
   A=LinearInterpolate(Alpha=LinearInterpolate(A=@n10, B=@n17.R, Alpha=@n27) {constAlpha=0.2}) {constA=-0.3}
   B=Add(A=VertexNormalWS(), B=VectorParameter() {parameterName=WPO_Direction, defaultValue=<rgba 0.3 0.3 1 0>}.RGB)
  )
  n29 = ScalarParameter() {defaultValue=4, parameterName=WPO_V_EdgeMask_Power}
  n30 = Sine(None=Multiply(A=@n12.R) {constB=0.5})
  n31 = ScalarParameter() {defaultValue=1, parameterName=WPO_U_EdgeMask_Power}
  n32 = Saturate(None=Multiply(A=Power(Base=@n24, Exp=@n29) {constExponent=5}, B=Power(Base=@n30, Exp=@n31) {constExponent=9}))
  Multiply(
   A=Power(Base=Multiply(A=@n28, B=@n32), Exp=ScalarParameter() {defaultValue=1.5, parameterName=WPO_Power}) {constExponent=1.5}
   B=ScalarParameter() {defaultValue=2, parameterName=WPO_Intensity}
  ) {constB=5}
 ]
}
