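# Niagara swing trail for the Hazard Edge pickaxe: a short red-hot burst of fire following the weapon's skeletal mesh.
# Trail_Flame and Trail_Glow build anim-trail ribbons, faded by emitter age and dimmed by time of day.
# Fire_Bits spawns flipbook flame sprites by rate and distance moved, partly dragging them with mesh motion as that influence decays.
# Embers adds socket-spawned sparks driven by inherited velocity, gravity and curl noise.
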NiagaraSystem /BRCosmetics/Weapons/FORT_Melee/Pickaxe_HazardEdge/FX/NS_HazardEdge_Pickaxe_Trail
dependencies {
 usedDataInterfaces [
  &NiagaraDataInterfaceCurve
  &NiagaraDataInterfaceColorCurve
  &NiagaraDataInterfaceSpriteRendererInfo
 ]
 usedDynamicInputs []
 usedModules [
  &self:SystemSpawnScript.SystemScriptSource.SystemScriptGraph.NiagaraNodeAssignment_0.SetVariables_#1
  &SystemState
  &self:SystemSpawnScript.SystemScriptSource.SystemScriptGraph.NiagaraNodeAssignment_1.SetVariables_#2
 ]
}
emitters {
 Embers {
  emitterUpdateScript [
   &EmitterState(Life Cycle Mode=::NewEnumerator1, Loop Behavior=::NewEnumerator1, Loop Delay=0.03, Loop Duration=0.2, UseLoopDelay=true)
   &SpawnRate(Spawn Probability=0.85, SpawnRate=67, Use Spawn Probability=true)
  ]
  particleSpawnScript [
   &InitializeParticle(
    Color=<rgba 4 4 4 1>
    Lifetime Max=0.41
    Lifetime Min=0.23
    Lifetime Mode=::NewEnumerator1
    Sprite Size Max=<2.4 6>
    Sprite Size Min=<1.4 3>
    Sprite Size Mode=::NewEnumerator2
    UsePositionOffset=false
    Write Material Random=true
   )
   &SkeletalMeshLocation(
    Bone / Socket Index=1
    MaskSpawnGroup=false
    Mesh Sampling Type=::NewEnumerator5
    Skeletal Mesh=$System.Skeletal Mesh
    Socket Sampling Mode=::NewEnumerator5
   )
   &ShapeLocation(
    MaskSpawnGroup=false
    Offset Mode=::NewEnumerator0
    Override Seed=false
    Rotation Axis=<0 1 0>
    Rotation Mode=::NewEnumerator1
    Sphere Radius=10
    Sphere Surface Distribution=0.7976
    UseOwnerScale=true
   )
   &AddVelocity(Velocity=&RandomRangeVector(Minimum=<0 0 80>, Maximum=<0 0 100>))
   &InheritVelocity(Speed Limit=true, UseSpeedThreshold=false)
  ]
  particleUpdateScript [
   &ParticleState()
   &ScaleSpriteSize(
    Uniform Curve Scale=1.1
    Uniform Curve Sprite Scale=DataInterface() {propertyValues {Curve {keys [<at 0 0 RCIM_Cubic>,<at 0.2 1 RCIM_Cubic>,<at 1 0 RCIM_Cubic>]}}}
   )
   &Drag(Drag=4, Use Linear Drag=true, Use Rotational Drag=true)
   /SetVariables(Particles.Glow03_Size=&Multiply_Vector2D_ByFloat(
     Float=&FloatFromCurve(
      FloatCurve=DataInterface() {propertyValues {Curve {keys [
          <at 0 0 RCIM_Cubic>
          <at 0.2 1.3 RCIM_Cubic>
          {arriveTangent=-0.1423, interpMode=RCIM_Cubic, leaveTangent=-0.1423, tangentMode=RCTM_User, time=1, value=0}
         ]}}}
      Scale Curve=11
     )
     Vector2D=<0.7 1.5>
    )) {isSetParameters=true}
   &ScaleColor(
    Scale Alpha=&FloatFromCurve(FloatCurve=DataInterface() {propertyValues {Curve {keys [
         {interpMode=RCIM_Cubic, tangentMode=RCTM_User, time=0.1, value=0}
         {arriveTangent=-1.3622, interpMode=RCIM_Cubic, leaveTangent=-1.3622, tangentMode=RCTM_User, time=0.2, value=1}
         {tangentMode=RCTM_User, time=1, value=0}
        ]}}})
    Scale RGB=&MakeVectorFromLinearColorRGB(Color=&ColorFromCurve(ColorCurve=DataInterface() {propertyValues {
        AlphaCurve {keys [<at 0.0926 1 RCIM_Cubic>]}
        BlueCurve {keys [<at 0 0.2138 RCIM_Cubic>,<at 0.2394 0.1336>,<at 0.922 0>]}
        GreenCurve {keys [<at 0 0.2138 RCIM_Cubic>,<at 0.2394 0.1336>,<at 0.922 0>]}
        RedCurve {keys [<at 0 8 RCIM_Cubic>,<at 0.2394 5>,<at 0.922 0.2>]}
       }}))
    ScaleA=true
    ScaleRGB=true
   )
   /SetVariables(Particles.Glow03_Color=&MakeLinearColorFromVectorAndFloat(
     "Vector (RGB)"=&MakeVectorFromLinearColorRGB(Color=&MultiplyLinearColor(Linear Color=<rgba 1.5 0 0 1>, Scale Factor=&MakeVector4_FromFloat(Float=1)))
     "Float (Alpha)"=&Multiply_Float(A=0.2, B=1)
    )) {isSetParameters=true}
   &GravityForce(Gravity=&RandomRangeVector(Minimum=<0 0 -150>, Maximum=<0 0 -250>))
   &CurlNoiseForce(
    Noise Frequency=200
    Noise Strength=&RandomRangeFloat(Minimum=800, Maximum=1100)
    Pan Noise=true
    Pan Noise Field=<0.2 0 1>
   )
   &SolveForcesAndVelocity(Clamp Velocity=false, Limit Acceleration=false)
  ]
  rendererClasses [&NiagaraSpriteRendererProperties]
  renderers [
   {rendererClass=&NiagaraSpriteRendererProperties}
   {rendererClass=&NiagaraSpriteRendererProperties, rendererIndex=1}
  ]
 }
 Fire_Bits {
  emitterUpdateScript [
   &EmitterState(Life Cycle Mode=::NewEnumerator1, Loop Behavior=::NewEnumerator1, Loop Duration=0.15, UseLoopDelay=false)
   &SpawnRate(
    Spawn Probability=&FloatFromCurve(FloatCurve=DataInterface() {propertyValues {Curve {keys [<at 0.06 1 RCIM_Constant>,<at 0.2 0 RCIM_Constant>]}}}, CurveIndex=$Emitter.Age)
    SpawnRate=16
    Use Spawn Probability=true
   )
   &SpawnPerUnit(
    Max Movement Threshold=4000
    Module.Use Max Movement Threshold=true
    Module.Use Movement Tolerance=false
    Spawn Probability:NiagaraFloat=&FloatFromCurve(FloatCurve=DataInterface() {propertyValues {Curve {keys [<at 0.06 1 RCIM_Constant>,<at 0.2 0 RCIM_Constant>]}}}, CurveIndex=$Emitter.Age)
    Spawn Spacing=16
    Use Spawn Probability=true
   )
  ]
  particleSpawnScript [&InitializeParticle(
    Color=<rgba 15 3 9 1>
    Lifetime Max=0.25
    Lifetime Min=0.2
    Lifetime Mode=::NewEnumerator1
    Position Offset=<0 0 50>
    Sprite Rotation Mode=::NewEnumerator2
    Sprite Size Mode=::NewEnumerator1
    Uniform Sprite Size Max=40
    Uniform Sprite Size Min=30
    UsePositionOffset=false
    Write Material Random=true
   )]
  particleUpdateScript [
   /SetVariables(Particles.PositionSkelMeshIntensity:NiagaraFloat=&Multiply_Float(
     A=&FloatFromCurve(
      FloatCurve=DataInterface() {propertyValues {Curve {keys [
          {arriveTangent=-0.961, interpMode=RCIM_Cubic, leaveTangent=-0.961, tangentMode=RCTM_User, time=0.14, value=1}
          {interpMode=RCIM_Cubic, tangentMode=RCTM_User, time=0.4, value=0}
         ]}}}
      Scale Curve=0.5
     )
     B=&FloatFromCurve(
      FloatCurve=DataInterface() {propertyValues {Curve {keys [{interpMode=RCIM_Cubic, tangentMode=RCTM_User, time=0.1, value=1}, {interpMode=RCIM_Cubic, tangentMode=RCTM_User, time=0.2, value=0}]}}}
      CurveIndex=$Emitter.Age
     )
    )) {isSetParameters=true}
   &SkeletalMeshLocation(
    Bone / Socket Index=2
    MaskSpawnGroup=false
    Mesh Sampling Type=::NewEnumerator5
    Position Sampling=::NewEnumerator1
    Skeletal Mesh=$System.Skeletal Mesh
    Socket Sampling Mode=::NewEnumerator5
    Source Movement Scale:Vector3f=&VectorFromFloat(Value=$Particles.PositionSkelMeshIntensity)
   )
   &ParticleState()
   &Drag(Drag=3, Use Linear Drag=true, Use Rotational Drag=true)
   &ScaleSpriteSize(Scale Factor=&Vector2DFromFloat(Value=&FloatFromCurve(
      FloatCurve=DataInterface() {propertyValues {Curve {keys [
          {arriveTangent=0.0388, interpMode=RCIM_Cubic, leaveTangent=0.0388, tangentMode=RCTM_User, time=0, value=0.6}
          {arriveTangent=-0.1015, interpMode=RCIM_Cubic, leaveTangent=-0.1015, tangentMode=RCTM_User, time=0.3, value=1.3}
          {arriveTangent=-1.0368, interpMode=RCIM_Cubic, leaveTangent=-1.0368, tangentMode=RCTM_User, time=1, value=0.7}
         ]}}}
      Scale Curve=0.8
     )))
   &ScaleColor(
    Scale Alpha=&FloatFromCurve(FloatCurve=DataInterface() {propertyValues {Curve {keys [
         {arriveTangent=3.5924, interpMode=RCIM_Cubic, leaveTangent=3.5924, tangentMode=RCTM_User, time=0, value=0}
         {arriveTangent=0.0151, interpMode=RCIM_Cubic, leaveTangent=0.0151, tangentMode=RCTM_User, time=0.1, value=1}
         <at 1 0 RCIM_Cubic>
        ]}}})
    Scale RGB=&MakeVectorFromLinearColorRGB(Color=&ColorFromCurve(ColorCurve=DataInterface() {propertyValues {
        AlphaCurve {keys [<at 0 1 RCIM_Cubic>]}
        BlueCurve {keys [<at 0 1.4>,<at 0.272 0>,<at 0.7468 0.0097>]}
        GreenCurve {keys [<at 0 1.68>,<at 0.272 0>,<at 0.7468 0>]}
        RedCurve {keys [<at 0 7>,<at 0.272 1>,<at 0.7468 0.2917>]}
       }}))
    ScaleA=true
    ScaleRGB=true
   )
   &ScaleColor(
    Scale Alpha=&FloatFromCurve(
     FloatCurve=DataInterface() {propertyValues {Curve {keys [{interpMode=RCIM_Cubic, tangentMode=RCTM_User, time=0, value=0}, {interpMode=RCIM_Cubic, tangentMode=RCTM_User, time=0.15, value=1}]}}}
     CurveIndex=$Emitter.Age
    )
    ScaleA=true
    ScaleRGB=false
   ) {module=ScaleColor001}
   &SolveForcesAndVelocity(Clamp Velocity=false, Limit Acceleration=false)
   &SubUVAnimation(
    End Frame Range Override:NiagaraInt32=&Add_Integer(A=$Particles.FlameStartingFrame, B=32)
    "Sprite Renderer "=DataInterface() {propertyValues {SpriteRenderer=&self:Fire_Bits_0.NiagaraSpriteRendererProperties_0}}
    Start Frame Range Override=$Particles.FlameStartingFrame
    UseEndFrame=true
    UseStartFrame=true
   ) {enabled=false}
   &DynamicMaterialParameters(
    Index 0 Param 1=2
    Module.Index0Param0Write=false
    Module.Index0Param1Write=false
    Module.Index0Param2Write=false
    Module.Index0Param3Write=false
    Module.Index1Param0Write=false
    Module.Index1Param1Write=false
    Module.Index1Param2Write=true
    Module.Index1Param3Write=false
    Write Parameter Index 1=true
   )
   &SubUVAnimation(
    "Sprite Renderer "=DataInterface() {propertyValues {SpriteRenderer=&self:Fire_Bits_0.NiagaraSpriteRendererProperties_0}}
    Start Frame Range Override=&RandomRangeInt(Minimum=40, Maximum=46)
    UseEndFrame=true
    UseStartFrame=true
   ) {module=SubUVAnimation001}
  ]
  rendererClasses [&NiagaraSpriteRendererProperties]
  renderers [{rendererClass=&NiagaraSpriteRendererProperties}]
 }
 Trail_Flame {
  emitterUpdateScript [
   &EmitterState(Life Cycle Mode=::NewEnumerator1, Loop Behavior=::NewEnumerator1, Loop Delay=1, Loop Duration=0.28, UseLoopDelay=false)
   &FN_SpawnAnimTrail()
  ]
  particleSpawnScript [
   &FN_InitializeAnimTrailBoneIndecies()
   &FN_InitializeAnimTrail(AnimTrailWidth=1.35, Lifetime=0.28, SkeletalMesh=$System.Skeletal Mesh)
   &AddVelocity(Velocity=<0 1000 0>) {enabled=false}
  ]
  particleUpdateScript [
   &SolveForcesAndVelocity(Clamp Velocity=false, Limit Acceleration=false) {enabled=false}
   &ParticleState()
   &Timeline(Play Rate=0, Timeline End=10000000)
   &FN_UpdateAnimTrail(SkeletalMesh=$System.Skeletal Mesh)
   &DynamicMaterialParameters(
    Index 0 Param 2=&FloatFromCurve(FloatCurve=DataInterface() {propertyValues {Curve {keys [<at 0 0>,<at 1 -0.7>]}}})
    Index 0 Param 4=&FloatFromCurve(FloatCurve=DataInterface() {propertyValues {Curve {keys [
         {arriveTangent=-0.2484, interpMode=RCIM_Cubic, leaveTangent=-0.2484, tangentMode=RCTM_User, time=0.5, value=1}
         {arriveTangent=-0.0843, interpMode=RCIM_Cubic, leaveTangent=-0.0843, tangentMode=RCTM_User, time=1, value=0}
        ]}}})
    Module.Index0Param0Write=true
    Module.Index0Param1Write=true
    Module.Index0Param2Write=false
    Module.Index0Param3Write=true
   )
   &ScaleColor(
    Scale Alpha=&FloatFromCurve(
     FloatCurve=DataInterface() {propertyValues {Curve {keys [
         <at 0 0 RCIM_Cubic>
         {interpMode=RCIM_Cubic, tangentMode=RCTM_User, time=0.06, value=1}
         {interpMode=RCIM_Cubic, tangentMode=RCTM_User, time=0.28, value=0}
        ]}}}
     CurveIndex=$Emitter.Age
     Scale Curve=&FNTimeOfDayScalar(Night Scale=0.55, Morning Scale=0.63, Day Scale=0.72, Evening Scale=0.63)
    )
    Scale RGB=&MakeVectorFromLinearColorRGB(Color=&ColorFromCurve(ColorCurve=DataInterface() {propertyValues {
        AlphaCurve {keys [<at 0 1 RCIM_Cubic>]}
        BlueCurve {keys [<at 0.0248 3 RCIM_Cubic>,<at 0.2077 1.2 RCIM_Cubic>]}
        GreenCurve {keys [<at 0.0248 3.85 RCIM_Cubic>,<at 0.2077 1.2 RCIM_Cubic>]}
        RedCurve {keys [<at 0.0248 20 RCIM_Cubic>,<at 0.2077 1.2 RCIM_Cubic>]}
       }}))
    ScaleA=true
    ScaleRGB=true
   )
   /SetVariables(Particles.Color2=&MakeLinearColorFromVectorAndFloat(
     "Vector (RGB)"=&MakeVectorFromLinearColorRGB(Color=&ColorFromCurve(ColorCurve=DataInterface() {propertyValues {
         AlphaCurve {keys [<at 0 1 RCIM_Cubic>]}
         BlueCurve {keys [<at 0 6.4 RCIM_Cubic>,<at 0.2579 1.5 RCIM_Cubic>,<at 1 0.7>]}
         GreenCurve {keys [<at 0 6.5333 RCIM_Cubic>,<at 0.2579 1.5 RCIM_Cubic>,<at 1 0.7>]}
         RedCurve {keys [<at 0 8 RCIM_Cubic>,<at 0.2579 1.5 RCIM_Cubic>,<at 1 0.7>]}
        }}))
     "Float (Alpha)"=&Multiply_Float(
      A=&FloatFromCurve(FloatCurve=DataInterface() {propertyValues {Curve {keys [
           {arriveTangent=-11.851, interpMode=RCIM_Cubic, leaveTangent=-11.851, tangentMode=RCTM_User, time=0, value=3}
           {arriveTangent=-0.6111, interpMode=RCIM_Cubic, leaveTangent=-0.6111, tangentMode=RCTM_User, time=0.2, value=0.7}
           {interpMode=RCIM_Cubic, tangentMode=RCTM_User, time=1, value=0.5}
          ]}}})
      B=&FloatFromCurve(
       FloatCurve=DataInterface() {propertyValues {Curve {keys [
           {arriveTangent=-4.1873, interpMode=RCIM_Cubic, tangentMode=RCTM_User, time=0.1, value=1}
           {interpMode=RCIM_Cubic, tangentMode=RCTM_User, time=0.31, value=0}
          ]}}}
       CurveIndex=$Emitter.Age
       Scale Curve=&FNTimeOfDayScalar(Night Scale=0.45, Morning Scale=0.5, Day Scale=0.6, Evening Scale=0.5)
      )
     )
    )) {isSetParameters=true}
   &CameraOffset(Camera Offset Amount=10)
  ]
  rendererClasses [&NiagaraRibbonRendererProperties]
  renderers [
   {rendererClass=&NiagaraRibbonRendererProperties}
   {rendererClass=&NiagaraRibbonRendererProperties, rendererIndex=1}
  ]
 }
 Trail_Glow {
  emitterUpdateScript [
   &EmitterState(Life Cycle Mode=::NewEnumerator1, Loop Behavior=::NewEnumerator1, Loop Delay=0.6, Loop Duration=1, UseLoopDelay=false)
   &FN_SpawnAnimTrail()
  ]
  particleSpawnScript [
   &FN_InitializeAnimTrailBoneIndecies()
   &FN_InitializeAnimTrail(AnimTrailWidth=1.15, Color=<rgba 1 0.02 0.0363 1>, Lifetime=0.33, SkeletalMesh=$System.Skeletal Mesh)
   &AddVelocity(Velocity=<0 1000 0>) {enabled=false}
  ]
  particleUpdateScript [
   &SolveForcesAndVelocity(Clamp Velocity=false, Limit Acceleration=false) {enabled=false}
   &ParticleState()
   &FN_UpdateAnimTrail(SkeletalMesh=$System.Skeletal Mesh)
   &ScaleRibbonWidth(Ribbon Width Scale=&FloatFromCurve(FloatCurve=DataInterface() {propertyValues {Curve {keys [
         {arriveTangent=0.9244, interpMode=RCIM_Cubic, leaveTangent=0.9244, tangentMode=RCTM_User, time=0, value=0.7}
         {arriveTangent=-0.8854, interpMode=RCIM_Cubic, leaveTangent=-0.8854, tangentMode=RCTM_User, time=0.1, value=1.2}
         {arriveTangent=-0.0997, interpMode=RCIM_Cubic, leaveTangent=-0.0997, tangentMode=RCTM_User, time=1, value=0.8}
        ]}}}))
   &ScaleColor(
    Scale Alpha=&FloatFromCurve(
     FloatCurve=DataInterface() {propertyValues {Curve {keys [
         {interpMode=RCIM_Cubic, tangentMode=RCTM_User, time=0, value=0}
         {arriveTangent=-3.9493, interpMode=RCIM_Cubic, leaveTangent=-3.9493, tangentMode=RCTM_User, time=0.1, value=1}
         {interpMode=RCIM_Cubic, tangentMode=RCTM_User, time=0.3, value=0}
        ]}}}
     CurveIndex=$Emitter.Age
     Scale Curve=&FNTimeOfDayScalar(Night Scale=0.75, Morning Scale=0.75, Evening Scale=0.8)
    )
    Scale RGB=&MakeVectorFromLinearColorRGB(Color=&ColorFromCurve(ColorCurve=DataInterface() {propertyValues {
        AlphaCurve {keys []}
        BlueCurve {keys [<at 0 0.8 RCIM_Cubic>,<at 0.2103 0.2133 RCIM_Cubic>]}
        GreenCurve {keys [<at 0 1.16 RCIM_Cubic>,<at 0.2103 0.2 RCIM_Cubic>]}
        RedCurve {keys [<at 0 8 RCIM_Cubic>,<at 0.2103 1 RCIM_Cubic>]}
       }}))
    ScaleA=true
    ScaleRGB=true
   )
   &ScaleColor(
    Scale Alpha=&FloatFromCurve(FloatCurve=DataInterface() {propertyValues {Curve {keys [
         {arriveTangent=-0.2537, interpMode=RCIM_Cubic, leaveTangent=-0.2537, tangentMode=RCTM_User, time=0, value=0}
         {arriveTangent=-0.2844, interpMode=RCIM_Cubic, leaveTangent=-0.2844, tangentMode=RCTM_User, time=0.1, value=1}
        ]}}})
    Scale RGB=&VectorFromFloat(Value=&FloatFromCurve(
      FloatCurve=DataInterface() {propertyValues {Curve {keys [
          {arriveTangent=-0.3629, interpMode=RCIM_Cubic, leaveTangent=-0.3629, tangentMode=RCTM_User, time=0, value=10}
          {arriveTangent=-4.7076, interpMode=RCIM_Cubic, leaveTangent=-4.7076, tangentMode=RCTM_User, time=0.22, value=1.8}
          <at 1 0.5 RCIM_Cubic>
         ]}}}
      Scale Curve=0.85
     ))
    ScaleA=true
    ScaleRGB=true
   ) {module=ScaleColor001}
   &Timeline(Play Rate=0, Timeline End=10000000) {enabled=false}
   &DynamicMaterialParameters(
    Index 0 Param 1=1
    Index 0 Param 3=1
    Module.Index0Param0Write=true
    Module.Index0Param1Write=false
    Module.Index0Param2Write=true
    Module.Index0Param3Write=true
   )
  ]
  rendererClasses [&NiagaraRibbonRendererProperties]
  renderers [{rendererClass=&NiagaraRibbonRendererProperties}]
 }
}
summary {
 systemName=NS_HazardEdge_Pickaxe_Trail
 emitters [
  {emitterName=Embers, rendererClasses [&NiagaraSpriteRendererProperties]}
  {emitterName=Fire_Bits, rendererClasses [&NiagaraSpriteRendererProperties]}
  {emitterName=Trail_Glow, rendererClasses [&NiagaraRibbonRendererProperties]}
  {emitterName=Trail_Flame, rendererClasses [&NiagaraRibbonRendererProperties]}
 ]
 userVariables [
  User.bIsFrontEnd {type=NiagaraBool}
  User.bIsFrontEndPreview {type=NiagaraBool}
  User.RibbonIDOffset {type=NiagaraInt32}
  User.Skeletal Mesh {
   type=NiagaraDataInterfaceSkeletalMesh
   defaultValue {
    struct=NiagaraExt_VariableValue_DataInterface
    value {dataInterface=&self:NiagaraDataInterfaceSkeletalMesh_0}
   }
  }
 ]
}
