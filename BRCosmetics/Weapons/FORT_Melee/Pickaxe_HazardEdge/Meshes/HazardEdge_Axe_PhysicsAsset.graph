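# Physics asset for the HazardEdge pickaxe melee weapon mesh in the cosmetics content.
# It defines a single capsule body on the Melee_TwoHanded_Handle bone, offset and tilted relative to that bone.
# With only that one body and no constraints, it gives the axe a simple rigid collision shape rather than any articulated simulation.
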PhysicsAsset /BRCosmetics/Weapons/FORT_Melee/Pickaxe_HazardEdge/Meshes/HazardEdge_Axe_PhysicsAsset
bodies {Melee_TwoHanded_Handle [{shapeType=Capsule, length=41.4259, radius=10.1839, center=<0.0005 3.781 23.7216>, rotation=<rot 5.2237 -90 -0.0009>}]}
constraints []
